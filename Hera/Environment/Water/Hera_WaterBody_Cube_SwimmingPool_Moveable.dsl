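; Moveable swimming pool cube water body that changes its water level shortly after play begins.
; On BeginPlay it waits a fixed delay, then plays a timeline from the start.
; Each timeline update lerps a vector ratio that differs only in Z and passes it to SetWaterRatio on the Fort water body cube volume component, raising the water from nearly empty to partly filled.
; The construction script is empty.
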(event EventBeginPlay
  (Utilities|FlowControl|Delay 12.0)
  (bind (_direction _lerp) (Timeline|Timeline PlayFromStart)
    (:Update
      (WaterCube|SetWaterRatio (Variables|WaterBody|GetFortWaterBodyCubeVolumeComponent) (Math|Vector|Lerp(Vector) "1.000000,1.000000,0.010000" "1.000000,1.000000,0.400000" _lerp) :!class /Script/FortniteGame.FortWaterBodyCubeVolumeComponent))))

(fn ConstructionScript)
