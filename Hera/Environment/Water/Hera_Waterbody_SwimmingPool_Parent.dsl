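; Parent water body Blueprint for Hera swimming pools that grants environmental healing to players in water when the actor is tagged TidePoolWater.
; It binds the pawn enter and exit water events on BeginPlay and unbinds them on EndPlay.
; Entering applies the tide pool healing GameplayEffect to Athena player pawns, and exiting removes it through the ability system component.
; Both handlers act only where cosmetic events cannot execute, such as dedicated servers.
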(event EventEndPlay
  (params (EndPlayReason :type EEndPlayReason%20Enum :default "Destroyed"))
  (|CallParentFunction EndPlayReason)
  (WaterBody|UnbindEventfromOnPawnEnterFortWaterBody (EventDispatchers|CreateEvent :!event Custom|OnPawnEnterWater))
  (WaterBody|UnbindEventfromOnPawnExitFortWaterBody (EventDispatchers|CreateEvent :!event Custom|OnPawnExitFortWater)))

(event EventBeginPlay
  (|CallParentFunction)
  (WaterBody|BindEventtoOnPawnEnterFortWaterBody (delegate Custom|OnPawnEnterWater))
  (WaterBody|BindEventtoOnPawnExitFortWaterBody (delegate Custom|OnPawnExitFortWater)))

(event Custom|OnPawnEnterWater
  (params (FortWaterBody :type Fort%20Water%20Body%20Actor%20Object%20Reference) (Player :type Fort%20Player%20Pawn%20Object%20Reference) (bIsFirstBody :type Boolean))
  (Networking|CanExecuteCosmeticEvents
    (:False
      (if (Actor|ActorHasTag "TidePoolWater")
        (Utilities|IsA(soft) Player "/Script/FortniteGame.FortPlayerPawnAthena"
          (:True
            (Environment|ApplyGlobalEnvironmentGameplayEffecttoActor Player "/Hera/Gameplay/TidePools/GE_Hera_TidePool_EnvHeal_Grant_Healing.GE_Hera_TidePool_EnvHeal_Grant_Healing_C")))))))

(event Custom|OnPawnExitFortWater
  (params (FortWaterBody :type Fort%20Water%20Body%20Actor%20Object%20Reference) (Player :type Fort%20Player%20Pawn%20Object%20Reference) (bIsLastBody :type Boolean))
  (bind _returnvalue (Ability|GetFortAbilitySystemComponent Player))
  (Networking|CanExecuteCosmeticEvents
    (:False
      (if (Actor|ActorHasTag "TidePoolWater")
        (Utilities|IsA(soft) Player "/Script/FortniteGame.FortPlayerPawnAthena"
          (:True
            (Utilities|IsValid _returnvalue
              (:Is%20Valid
                (GameplayEffects|RemoveActiveGameplayEffectbySourceEffect _returnvalue "/Hera/Gameplay/TidePools/GE_Hera_TidePool_EnvHeal_Grant_Healing.GE_Hera_TidePool_EnvHeal_Grant_Healing_C" 0 -1)))))))))

(fn ConstructionScript)
