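# Material function for mobile water surfaces that clips geometry to the water zone bounds by adjusting an input offset vector.
# It reads zone location and extent from WaterZoneData, clamps world XY to that box and subtracts the original position, giving an offset that pulls outside points onto the boundary.
# This XY offset replaces the input's XY while keeping its Z.
# The static switch "Bounds Clip Outside" selects this result or passes the input through.
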MaterialFunction /FortWater/Materials/WaterSurface/Mobile/Functions/MF_MobileBoundsClip
graph {nodes [
  Constant2Vector()
  n1 = ComponentMask(None=WorldPosition().XYZ) {g=true, r=true}
  n2 = MaterialFunctionCall() {materialFunction=/Water/Materials/Functions/WaterZoneData}
  n3 = Divide(A=@n2.Extent)
  n4 = Add(
   A=Clamp(None=@n1, Min=Subtract(A=@n2.Location, B=@n3), Max=Add(A=@n2.Location, B=@n3)) {maxDefault=1000000, minDefault=-1000000}
   B=Multiply(A=@n1) {constB=-1}
  )
  n5 = FunctionInput()
  n6 = ComponentMask(None=@n5) {b=true}
  FunctionOutput(None=StaticSwitchParameter(True=AppendVector(A=@n4, B=@n6), False=@n5) {parameterName=Bounds Clip Outside})
 ]}
