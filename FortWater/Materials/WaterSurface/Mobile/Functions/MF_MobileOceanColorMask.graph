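# Material function for the mobile ocean surface that outputs a tinted color mask over a circular region of the world.
# World XY distance from a parameterized origin, past a radius, drives a falloff sharpened by repeated squaring, then reduced by the inverted Island Mask input.
# World-projected ocean noise breaks up that falloff through a multiply-add blend.
# The saturated result is multiplied by the MaskColor parameter to produce the Mask Color output.
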MaterialFunction /FortWater/Materials/WaterSurface/Mobile/Functions/MF_MobileOceanColorMask
graph {nodes [
  n1 = WorldPosition()
  n2 = Distance(
   A=@n1.XY
   B=AppendVector(
    A=ScalarParameter() {defaultValue=-19930.2812, group=OceanColorMask, parameterName=MaskOriginX}
    B=ScalarParameter() {defaultValue=45887.2422, group=OceanColorMask, parameterName=MaskOriginY}
   )
  )
  n3 = ScalarParameter() {defaultValue=42523.7188, group=OceanColorMask, parameterName=MaskRadius}
  n4 = Saturate(None=Divide(A=Subtract(A=@n2, B=@n3), B=ScalarParameter() {defaultValue=267046.6562, group=OceanColorMask, parameterName=MaskFade}))
  n5 = OneMinus(None=Multiply(A=@n4, B=@n4))
  n6 = Multiply(A=@n5, B=@n5)
  n7 = Multiply(A=@n6, B=@n6)
  n8 = Multiply(A=Multiply(A=@n7, B=@n7), B=OneMinus(None=FunctionInput() {inputName=Island Mask, inputType=FunctionInput_Scalar}))
  n9 = TextureSample(UVs=Divide(A=@n1.XY, B=ScalarParameter() {defaultValue=100000, group=OceanColorMask, parameterName=TextureSize})) {samplerSource=SSM_Wrap_WorldGroupSettings, samplerType=SAMPLERTYPE_Masks, texture=/FortWater/Textures/OceanNoise__Mask}
  n10 = Multiply(A=@n8, B=Add(A=@n8, B=LinearInterpolate(Alpha=@n9.R) {constA=0.1, constB=0.9}))
  FunctionOutput(None=Multiply(A=Saturate(None=@n10), B=VectorParameter() {parameterName=MaskColor, defaultValue=<rgba 0 0.25 0.1959 1>}.RGB)) {outputName=Mask Color}
 ]}
