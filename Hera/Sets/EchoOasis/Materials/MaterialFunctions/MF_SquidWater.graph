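# Water material function that paints stylized, animated squid water with a spiral around a world-space pivot in Echo Oasis.
# Depth- and noise-driven sine rings blend into an arctangent spiral, coloring scattering and bending flow-mapped normals.
# Two hard-coded world-space segment gradients mask where the effect overrides scattering, absorption, and mobile emissive and opacity.
# Static switches gate those overrides, but not Normal, and select a time-of-day-scaled mobile color path.
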MaterialFunction /Hera/Sets/EchoOasis/Materials/MaterialFunctions/MF_SquidWater
graph {
 properties {bExposeToLibrary=true}
 nodes [
  n1 = ScalarParameter() {defaultValue=20, parameterName=Rings}
  n2 = Saturate(None=Divide(
    A=MaterialFunctionCall!(Position) {materialFunction=/Water/Materials/Functions/WaterBodyData}.Water Depth
    B=ScalarParameter() {defaultValue=400, parameterName=Waterbody Depth Scale}
   ) {constB=100})
  n3 = ScalarParameter() {defaultValue=4, parameterName=Mip1}
  n4 = Multiply(A=ScalarParameter() {defaultValue=0.3, parameterName=Mip1 Tiling Adjustment}, B=@n3)
  n5 = WorldPosition()
  n6 = Constant() {r=0.02}
  n7 = Time()
  n8 = Constant2Vector() {g=-0.3, r=-1}
  n9 = Add(A=Multiply(A=Add(A=@n4), B=Divide(A=@n5.XY) {constB=2000}), B=Multiply(A=@n8, B=Multiply(A=@n7, B=@n6)))
  n10 = TextureSample!(UVs=@n9, Tex, Level=@n3, Apply View MipBias) {
   mipValueMode=TMVM_MipLevel
   samplerSource=SSM_Wrap_WorldGroupSettings
   samplerType=SAMPLERTYPE_Masks
   texture=&/T_SquidWater_Noise
  }
  n11 = Multiply(A=ScalarParameter() {defaultValue=0.05, parameterName=WBD Noise Intensity}, B=@n10.R)
  n12 = ScalarParameter() {defaultValue=0.5, parameterName=CenterHoleIntensity}
  n13 = Time()
  n14 = Constant()
  n15 = ComponentMask(None=Subtract(A=WorldPosition().XYZ, B=VectorParameter() {parameterName=SpiralPivot, defaultValue=<rgba 60300 34200 250 1>}.RGB)) {g=true, r=true}
  n16 = Divide(A=@n15) {constB=5000}
  n17 = Subtract(A=Saturate(None=Add(A=@n16) {constB=0.5}), B=Constant() {r=0.5})
  n18 = Multiply(A=@n17, B=@n17)
  n19 = Add(A=ComponentMask(None=@n18) {r=true}, B=ComponentMask(None=@n18) {g=true})
  n20 = Multiply(A=Add(A=Multiply(A=@n13, B=@n14), B=SquareRoot(None=@n19))) {constB=2}
  n21 = Saturate(None=OneMinus(None=@n20))
  n22 = Multiply(A=ScalarParameter() {defaultValue=1, parameterName=CenterHoleSize}, B=OneMinus(None=@n21))
  n23 = MaterialFunctionCall(Alpha=Saturate(None=OneMinus(None=@n22))) {materialFunction=&SmoothStep}
  n24 = ScalarParameter() {defaultValue=0.5, parameterName=Blend factor centerHole}
  n25 = MaterialFunctionCall(In=@n23.Result, Contrast=Constant() {r=0.2}) {materialFunction=&CheapContrast}
  n26 = Divide(A=LinearInterpolate(A=Add(A=@n2, B=@n11), B=Multiply(A=@n12, B=@n23.Result), Alpha=Multiply(A=@n24, B=@n25.Result)) {constAlpha=0.9}) {constB=5}
  n27 = Constant() {r=0.03}
  n28 = Time()
  n29 = Sine(None=Add(A=Multiply(A=@n1, B=@n26), B=Multiply(A=@n27, B=@n28)))
  n30 = Multiply(A=@n7, B=Constant() {r=0.05})
  n31 = TextureSample(UVs=Add(A=Divide(A=@n5.XY) {constB=5000}, B=Multiply(A=@n8, B=@n30)), Level=ScalarParameter() {defaultValue=4, parameterName=Mip2}) {
   mipValueMode=TMVM_MipLevel
   samplerSource=SSM_Wrap_WorldGroupSettings
   samplerType=SAMPLERTYPE_Masks
   texture=&T_SquidWater_Noise
  }
  n32 = LinearInterpolate(A=Sine(None=@n29), B=Cosine(None=@n29), Alpha=@n31.R)
  n33 = DDX(Value=@n32)
  n34 = ComponentMask(None=DDX(Value=@n5.XY)) {r=true}
  n35 = DDY(Value=@n32)
  n36 = ComponentMask(None=DDX(Value=@n5.XY)) {g=true}
  AppendVector(A=Divide(A=@n33, B=Add(A=@n34) {constB=0}), B=Divide(A=@n35, B=Add(A=@n36) {constB=0}))
  Constant() {r=4}
  Constant() {r=1}
  Constant() {r=0.025}
  Constant() {r=-0.3}
  Constant() {r=2}
  n37 = FunctionInput() {bUsePreviewValueAsDefault=true, inputName=Scattering, previewValue=<1 1 1 1>}
  n38 = MaterialFunctionCall(
   Value_Morning=ScalarParameter() {defaultValue=1, parameterName=Mobile - Color Intensity - Morning}
   Value_Day=ScalarParameter() {defaultValue=1, parameterName=Mobile - Color Intensity - Day}
   Value_Evening=ScalarParameter() {defaultValue=1, parameterName=Mobile - Color Intensity - Evening}
   Value_Night=ScalarParameter() {defaultValue=0.5, parameterName=Mobile - Color Intensity - Night}
  ) {materialFunction=&Scalar_Time_of_Day}
  n39 = Constant4Vector() {constant=<rgba 0.3445 0.1061 0.5625 1>}
  n40 = Constant4Vector() {constant=<rgba 0.3069 0.1201 0.5625 1>}
  n41 = ScalarParameter() {defaultValue=1, parameterName=Stripes gradient offset}
  n42 = ComponentMask(None=Arctangent2!(Y=ComponentMask(None=@n17) {r=true}, X=ComponentMask(None=@n17) {g=true})) {r=true}
  n43 = MaterialFunctionCall(Multiplier=Constant() {r=2}) {materialFunction=&Pi}
  n44 = Add(A=Frac(None=Divide(A=@n42, B=@n43.Result)), B=Multiply(A=Constant() {r=-0.3}, B=@n13))
  n45 = Add(A=@n20) {constB=-0.05}
  n46 = Frac(None=Add(A=Multiply(A=@n44), B=Multiply(A=@n45) {constB=15}))
  n47 = Sine(None=Frac(None=Multiply(A=@n46)))
  n48 = Multiply(A=OneMinus(None=Multiply(A=@n41, B=@n47)), B=OneMinus(None=Power(Base=@n21) {constExponent=22}))
  n49 = LinearInterpolate(
   A=@n32
   B=MaterialFunctionCall(Alpha=Multiply(A=@n48) {constB=0.5}) {materialFunction=&SmoothStep}.Result
   Alpha=Power(Base=@n21) {constExponent=5}
  )
  n50 = Power(Base=OneMinus(None=@n49)) {constExponent=5}
  n51 = Saturate(None=Add(A=OneMinus(None=@n50)) {constB=-0.075})
  n52 = ScalarParameter() {parameterName=Blue Line Intensity RIver}
  n53 = ScalarParameter() {defaultValue=0.99, parameterName=goffset}
  n54 = ScalarParameter() {defaultValue=0.1, parameterName=gsize}
  n55 = Power(Base=MaterialFunctionCall(In=@n10.R, Contrast=Constant() {r=0.5}) {materialFunction=&CheapContrast}.Result)
  n56 = Add(A=Multiply(A=Sine(None=@n55)) {constB=0.08}, B=@n51)
  n57 = MaterialFunctionCall(Input=@n56, Input Low=Subtract(A=@n53, B=@n54), Input High=Add(A=@n53, B=@n54)) {materialFunction=&RemapValue_0-1}
  n58 = Power(Base=Sine(None=@n57.Result)) {constExponent=5}
  n59 = Multiply(
   A=Multiply(A=@n51, B=MaterialFunctionCall(Alpha=@n58) {materialFunction=&SmoothStep}.Result)
   B=@n52
  )
  n60 = MaterialFunctionCall() {materialFunction=&WaterBodyData}
  n61 = Constant3Vector() {desc=A Coord, constant=<rgba 65296 35556 150 0>}
  n62 = Subtract(A=Constant3Vector() {desc=B Coord, constant=<rgba 73800 33476 150 0>}, B=@n61) {desc=AB}
  n63 = DotProduct(A=Subtract(A=WorldPosition().XYZ, B=@n61) {desc=AP}, B=@n62)
  n64 = DotProduct(A=@n62, B=@n62)
  n65 = OneMinus(None=Saturate(None=Divide(A=@n63, B=@n64)))
  n66 = Constant3Vector() {desc=B Coord, constant=<rgba 58296 33732 300 0>}
  n67 = Subtract(A=Constant3Vector() {desc=A Coord, constant=<rgba 57488 33144 300 0>}, B=@n66) {desc=AB}
  n68 = DotProduct(A=Subtract(A=WorldPosition().XYZ, B=@n66) {desc=AP}, B=@n67)
  n69 = DotProduct(A=@n67, B=@n67)
  n70 = OneMinus(None=Saturate(None=Divide(A=@n68, B=@n69)))
  n71 = Power(
   Base=MaterialFunctionCall(Alpha=Multiply(A=@n65, B=@n70)) {materialFunction=&SmoothStep}.Result
   Exp=ScalarParameter() {defaultValue=1, parameterName=Downstream Power}
  ) {constExponent=5}
  n72 = Multiply(A=Saturate(None=Divide(A=@n60.Water Depth) {constB=100}), B=@n71)
  n73 = LinearInterpolate(
   A=Constant4Vector() {constant=<rgba 0.1667 0 1 1>}.RGBA
   B=LinearInterpolate(A=LinearInterpolate(A=@n39.RGBA, B=@n40.RGBA, Alpha=@n51), B=Constant4Vector() {constant=<rgba 0 0.6667 1 1>}.RGBA, Alpha=@n59)
   Alpha=Saturate(None=Multiply(A=@n72) {constB=5})
  )
  n74 = LinearInterpolate(
   A=LinearInterpolate(
    A=VectorParameter() {parameterName=Color A, defaultValue=<rgba 0.5 0 1 1>}.RGB
    B=VectorParameter() {parameterName=Color B, defaultValue=<rgba 0.6667 0 1 1>}.RGB
    Alpha=@n51
   )
   B=VectorParameter() {parameterName=Color B2, defaultValue=<rgba 0 0.6667 1 1>}.RGB
   Alpha=@n59
  )
  n75 = StaticSwitchParameter(
   True=ComponentMask(None=Multiply(A=@n38.Result, B=@n73)) {b=true, g=true, r=true}
   False=LinearInterpolate(A=VectorParameter() {parameterName=Color C, defaultValue=<rgba 0.0167 0 0.1 1>}.RGB, B=Power(Base=@n74), Alpha=@n72)
  ) {parameterName=UseSquidWaterMobile}
  n76 = Multiply(A=@n75, B=ScalarParameter() {defaultValue=1, parameterName=SquidWater Scattering Multiplier})
  FunctionOutput(None=StaticSwitchParameter(True=LinearInterpolate(A=@n37, B=@n76, Alpha=@n71), False=@n37) {parameterName=UseSquidWater}) {outputName=Scattering}
  n77 = FunctionInput() {bUsePreviewValueAsDefault=true, inputName=Absorption, sortPriority=1, previewValue=<1 1 1 1>}
  n78 = Multiply(A=@n77, B=ScalarParameter() {defaultValue=0.1, parameterName=SquidWater Absorption Multiplier})
  FunctionOutput(None=StaticSwitchParameter(True=LinearInterpolate(A=@n77, B=@n78, Alpha=@n71), False=@n77) {parameterName=UseSquidWater}) {outputName=Absorption, sortPriority=1}
  n79 = OneMinus(None=@n21)
  n80 = MaterialFunctionCall(Alpha=Saturate(None=Multiply(A=@n79) {constB=10})) {materialFunction=&SmoothStep}
  n81 = Multiply(A=ScalarParameter() {defaultValue=-1, parameterName=Squid Normal Intensity 2}, B=@n80.Result)
  n82 = Normalize(VectorInput=@n15)
  n83 = Normalize(VectorInput=AppendVector(A=Multiply(A=@n81, B=@n82), B=Constant() {r=1}))
  n84 = Multiply(A=Constant3Vector() {constant=<rgba -1 -1 1 0>}, B=@n83) {constB=-1}
  n85 = Saturate(None=Power(Base=@n21) {constExponent=10})
  n86 = Multiply(A=ScalarParameter() {defaultValue=1, parameterName=Squid Normal Intensity}, B=Multiply(A=Add(A=@n49) {constB=-0.5}) {constB=2})
  n87 = Multiply(
   A=LinearInterpolate(
    A=ScalarParameter() {defaultValue=0.1, parameterName=Squid Normal Intensity NoSpiral}
    B=ScalarParameter() {defaultValue=1, parameterName=Squid Normal Intensity Spiral}
    Alpha=Power(Base=@n21)
   )
   B=@n71
  )
  n88 = Multiply(A=Multiply(A=@n80.Result, B=Multiply(A=@n86, B=@n87)), B=@n82)
  n89 = Constant() {r=1}
  n90 = MaterialFunctionCall(
   BaseNormal=LinearInterpolate(A=Constant3Vector() {constant=<rgba 0 0 1 0>}, B=LinearInterpolate(A=@n84, B=@n83, Alpha=@n85), Alpha=@n85)
   AdditionalNormal=Normalize(VectorInput=AppendVector(A=@n88, B=@n89))
  ) {materialFunction=&BlendAngleCorrectedNormals}
  n91 = ScalarParameter() {defaultValue=0.25, parameterName=nint}
  n92 = Multiply(A=AppendVector(A=@n91, B=@n91), B=OneMinus(None=@n21))
  n93 = Normalize(VectorInput=AppendVector(A=Multiply(A=@n92, B=@n71), B=Constant() {r=1}))
  n94 = Constant2Vector() {g=-1, r=1}
  n95 = MaterialFunctionCall!(
   Diffuse
   Normal Map=TextureObject() {samplerType=SAMPLERTYPE_Normal, texture=&/T_SquidWater_N}
   "Flow Vector Map (see tooltip)"=Multiply(A=Multiply(A=@n82, B=@n94)) {constB=0.1}
   UV texture 2 offset
   "Time (see tooltip)"
   useMipLevel=StaticBool() {value=true}
   UVs=Multiply(A=@n9, B=ScalarParameter() {defaultValue=1.5, parameterName=Normal Tiling})
   Texture Size
   Texture Mip Bias=ScalarParameter() {parameterName=mipp}
  ) {materialFunction=/Engine/Functions/Engine_MaterialFunctions02/Texturing/Flowmaps.FlowMaps}
  FunctionOutput(None=MaterialFunctionCall(BaseNormal=Normalize(VectorInput=@n90.Result), AdditionalNormal=Multiply(A=@n93, B=@n95.Normal)) {materialFunction=&BlendAngleCorrectedNormals}.Result) {outputName=Normal, sortPriority=1}
  n96 = FunctionInput() {bUsePreviewValueAsDefault=true, inputName=MobileEmissive, previewValue=<1 1 1 1>}
  n97 = ScalarParameter() {defaultValue=0.8, parameterName=Mobile Reflection Add}
  n98 = ScalarParameter() {defaultValue=5, parameterName=Mobile Sun Intensity}
  n99 = FunctionInput() {inputName=Sun Specular, sortPriority=20}
  n100 = LinearInterpolate(
   A=@n75
   B=Add(A=Add(A=@n97, B=@n75) {constA=0.4}, B=Multiply(A=@n98, B=@n99))
   Alpha=FunctionInput() {inputName=Is translucent, inputType=FunctionInput_Scalar, sortPriority=30}
  )
  FunctionOutput(None=StaticSwitchParameter(True=LinearInterpolate(A=@n96, B=@n100, Alpha=@n71), False=@n96) {parameterName=UseSquidWater}) {outputName=MobileEmissive}
  n101 = FunctionInput() {bUsePreviewValueAsDefault=true, inputName=MobileOpacity, previewValue=<1 1 1 1>}
  n102 = DepthFade(FadeDistance=ScalarParameter() {defaultValue=20, parameterName=Mobile Depth Fade})
  n103 = FunctionInput() {inputName=Material Attributes, inputType=FunctionInput_MaterialAttributes, sortPriority=-10}
  n104 = Multiply(
   A=Max(A=GetMaterialAttributes(None=@n103) {attributeGetTypes [#1, #2]}.Opacity) {constB=0.75}
   B=@n71
  )
  FunctionOutput(None=StaticSwitchParameter(True=LinearInterpolate(A=@n101, B=@n102, Alpha=@n104), False=@n101) {parameterName=UseSquidWater}) {outputName=MobileOpacity}
  LinearInterpolate(A=@n52, B=ScalarParameter() {defaultValue=0.25, parameterName=Blue Line Intensity}, Alpha=Power(Base=@n71) {constExponent=222})
  MaterialFunctionCall(Value=@n57.Result) {materialFunction=&LinearSine}
  n105 = Multiply(A=Distance(A=Frac(None=@n56), B=Constant() {r=0.6}))
  Multiply(A=Multiply(A=Power(Base=@n105)) {constB=2}, B=@n51)
  n106 = ScalarParameter() {defaultValue=5, parameterName=DDXY Intensity}
  n107 = AppendVector(A=@n33, B=@n35)
  Normalize(VectorInput=AppendVector(A=Multiply(A=@n106, B=@n107), B=Constant() {r=1}))
  Saturate(None=@n32)
  ScalarParameter() {defaultValue=0.8, parameterName=Mobile Depth Opacity}
 ]
}
