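# Translucent unlit mobile water material with static switches for ocean, river, lake, pool and transitions between water types.
# Builds depth-based color, opacity and foam from scrolling textures and two-phase river flow, adding cubemap reflections, sun specular and caustics.
# Water body data drives wave world position offset with bounds clipping.
# Attributes pass through MF_Water_Styx and MF_LiquidGold, then MF_SquidWater supplies values to a final attribute override.
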Material /Hera/Environment/Water/M_MobileWater_Squid
graph {
 properties {
  allowTranslucentCustomDepthWrites=true
  bApplyCloudFogging=true
  bComputeFogPerPixel=true
  bTangentSpaceNormal=false
  bUseMaterialAttributes=true
  bUsedWithWater=true
  blendMode=BLEND_Translucent
  pixelDepthOffsetMode=PDOM_Legacy
  shadingModel=MSM_Unlit
  usedShadingModels=Unlit
 }
 nodes [
  n1 = NamedRerouteUsage()
  n2 = TwoSidedSign()
  Add(A=@n1, B=OneMinus(None=Saturate(None=@n2)))
  n3 = LinearInterpolate(A=Constant3Vector(), B=NamedRerouteUsage(), Alpha=NamedRerouteUsage())
  ComponentMask(None=@n3) {r=true}
  Constant() {r=1}
  n4 = NamedRerouteUsage()
  n5 = ScalarParameter() {defaultValue=0.003, group=Mobile Water, parameterName=MobilePool Color Depth}
  n6 = LinearInterpolate(B=Saturate(None=Multiply(A=@n4, B=@n5) {constB=0.003}), Alpha=NamedRerouteUsage()) {constA=1}
  n7 = OneMinus(None=NamedRerouteUsage())
  n8 = NamedRerouteUsage()
  n9 = NamedRerouteUsage()
  n10 = Multiply(A=Add(A=NamedRerouteUsage(), B=NamedRerouteUsage()), B=NamedRerouteUsage())
  n11 = StaticSwitchParameter(
   True=LinearInterpolate(A=NamedRerouteUsage(), B=NamedRerouteUsage(), Alpha=Multiply(A=@n6, B=@n7))
   False=StaticSwitchParameter(True=@n8, False=LinearInterpolate(A=@n8, B=@n9, Alpha=@n10)) {group=Mobile Water Switches, parameterName=Mobile Ocean Extent}
  ) {group=Mobile Water Switches, parameterName=Mobile Pool}
  n12 = NamedRerouteUsage()
  n13 = StaticSwitchParameter(True=LinearInterpolate(A=@n11, B=@n12, Alpha=NamedRerouteUsage()), False=@n12) {group=Gold Liquid_Blend, parameterName=UseGold_SphereMask_Blend}
  n14 = Multiply(
   A=LinearInterpolate(
    A=StaticSwitchParameter(True=@n13, False=@n11) {group=Gold Liquid, parameterName=Gold Liquid}
    B=NamedRerouteUsage()
    Alpha=NamedRerouteUsage()
   ) {constAlpha=0}
   B=NamedRerouteUsage()
  )
  n15 = NamedRerouteUsage()
  n16 = Multiply(
   A=LinearInterpolate(
    A=VectorParameter() {group=Mobile Water, parameterName=Mobile Foam Color 1, defaultValue=<rgba 0.4485 0.6949 0.8698 1>}.RGB
    B=VectorParameter() {group=Mobile Water, parameterName=Mobile Foam Color 2, defaultValue=<rgba 0.7548 0.7642 0.7708 1>}.RGB
    Alpha=Multiply(A=@n15, B=@n15)
   )
   B=NamedRerouteUsage()
  )
  n17 = Multiply(A=@n15, B=OneMinus(None=NamedRerouteUsage()))
  n18 = StaticSwitchParameter(True=@n17, False=@n15) {parameterName=UseGold_SphereMask_Blend}
  n19 = StaticSwitchParameter(
   True=@n14
   False=StaticSwitchParameter(True=@n14, False=LinearInterpolate(A=@n14, B=@n16, Alpha=@n18)) {group=Mobile Water Switches, parameterName=Mobile Ocean Extent}
  ) {group=Mobile Water Switches, parameterName=Mobile Pool}
  n20 = NamedRerouteUsage()
  n21 = ScalarParameter() {defaultValue=0.88, group=Mobile Water, parameterName=MobilePool Opacity Deep}
  n22 = Saturate(None=Multiply(A=NamedRerouteUsage(), B=ScalarParameter() {defaultValue=0.05, group=Mobile Water, parameterName=MobilePool Opacity Depth}) {constB=0.05})
  n23 = NamedRerouteUsage()
  n24 = NamedRerouteUsage()
  n25 = Add(
   A=NamedRerouteUsage()
   B=LinearInterpolate(
    A=ScalarParameter() {defaultValue=0.88, group=Mobile Water, parameterName=Mobile Opacity Deep}
    B=ScalarParameter() {defaultValue=0.3, group=Mobile Water, parameterName=Mobile Opacity Shallow}
    Alpha=Multiply(A=@n24, B=@n24)
   ) {constA=0.92, constB=0.5}
  )
  n26 = NamedRerouteUsage()
  n27 = NamedRerouteUsage()
  n28 = LinearInterpolate(B=Saturate(None=Multiply(A=@n27) {constB=0.02}), Alpha=NamedRerouteUsage()) {constA=1}
  n29 = Saturate(None=Add(A=NamedRerouteUsage()) {constB=10})
  n30 = StaticSwitchParameter(
   True=Multiply(A=Add(A=@n20, B=@n21), B=LinearInterpolate(B=@n22, Alpha=@n23) {constA=1})
   False=Multiply(A=LinearInterpolate(B=@n25, Alpha=@n26) {constA=1}, B=Multiply(A=@n28, B=@n29))
  ) {group=Mobile Water Switches, parameterName=Mobile Pool}
  n31 = SetMaterialAttributes!(
   MaterialAttributes
   Emissive Color=Add(A=StaticSwitchParameter(True=@n14, False=@n19) {group=Mobile Water Switches, parameterName=DisableFoam}, B=NamedRerouteUsage())
   Opacity=StaticSwitchParameter(True=Constant() {r=1}, False=@n30) {group=Mobile Water Switches, parameterName=Mobile Ocean Extent}
   World Position Offset=NamedRerouteUsage()
  ) {attributeSetTypes [#1, #2, #3]}
  n32 = MaterialFunctionCall!(
   DefaultMaterial=@n31
   Material_NoFoam=@n31
   Depth=NamedRerouteUsage()
   SimWaterNormal
   SimErasure
   TransitionMask
   Scattering
   Absorption
   Is Translucent Mat=StaticBool() {value=true}
   Mobile Reflections=LinearInterpolate(A=Constant3Vector(), B=NamedRerouteUsage(), Alpha=NamedRerouteUsage())
  ) {materialFunction=/FortWater/Materials/Functions/MF_Water_Styx}
  n33 = MaterialFunctionCall!(
   In=@n32.Result
   Velocity=NamedRerouteUsage()
   Depth=NamedRerouteUsage()
   Transition=NamedRerouteUsage()
   IsRiverMaterial=StaticBoolParameter() {parameterName=Mobile River Gold}
   IsTranslucentMat=StaticBool() {value=true}
  ) {materialFunction=/FortWater/Materials/WaterSurface/Hermes/Gold/MF_LiquidGold}
  n34 = GetMaterialAttributes(None=@n33.Result) {attributeGetTypes [#1, #4, #2]}
  MaterialFunctionCall(BaseNormal=@n34.Normal) {materialFunction=&BlendAngleCorrectedNormals}
  n35 = TwoSidedSign()
  Multiply(A=@n30, B=Saturate(None=Add(A=@n35) {constB=1.5}))
  Multiply(B=OneMinus(None=NamedRerouteUsage()))
  n36 = MaterialFunctionCall!(Texture, Position) {materialFunction=/FortWater/Materials/Functions/WaterBodyData}
  NamedRerouteDeclaration(None=@n36.Water Depth) {name=Water Depth, nodeColor=<rgba 0 0.0149 0.1094 1>}
  n37 = Subtract(A=SceneDepth().R, B=PixelDepth().R)
  n38 = NamedRerouteDeclaration(None=Multiply(A=CameraVectorWS(), B=NamedRerouteDeclaration(None=@n37) {name=Screen Space Depth, nodeColor=<rgba 0 1 0.9531 1>})) {name=WPO Depth, nodeColor=<rgba 1 0 0.9375 1>}
  n39 = ComponentMask(None=Add(A=WorldPosition() {worldPositionShaderOffset=WPT_ExcludeAllShaderOffsets}.XYZ, B=Multiply(A=@n38) {constB=-1})) {g=true, r=true}
  n40 = ScalarParameter() {defaultValue=0.0009, group=Mobile Water, parameterName=Mobile Caustics1 Scale}
  n41 = Time()
  n42 = ScalarParameter() {defaultValue=0.04, group=Mobile Water, parameterName=Mobile Caustics1 Speed}
  n43 = TextureSample(UVs=Add(A=Multiply(A=@n39, B=@n40) {constB=0.0008}, B=Multiply(A=@n41, B=@n42) {constB=0.03})) {samplerType=SAMPLERTYPE_LinearGrayscale, texture=/FortWater/Materials/WaterSurface/Mobile/Textures/T_MobileWater_Caustics}
  n44 = AppendVector(A=ComponentMask(None=@n39) {g=true}, B=ComponentMask(None=@n39) {r=true})
  n45 = Multiply(A=ScalarParameter() {defaultValue=0.0008, group=Mobile Water, parameterName=Mobile Caustics2 Scale}, B=Constant2Vector() {g=1, r=0.5})
  n46 = Time()
  n47 = ScalarParameter() {defaultValue=0.05, group=Mobile Water, parameterName=Mobile Caustics2 Speed}
  n48 = TextureSample(UVs=Subtract(A=Multiply(A=@n44, B=@n45) {constB=0.0007}, B=Multiply(A=@n46, B=@n47) {constB=0.02})) {samplerType=SAMPLERTYPE_LinearGrayscale, texture=&T_MobileWater_Caustics}
  n49 = Saturate(None=Divide(A=PixelDepth().R, B=ScalarParameter() {defaultValue=30000, group=Mobile Water, parameterName=Mobile Caustics Fade Distance}) {constB=10000})
  n50 = Multiply(
   A=Multiply(A=OneMinus(None=@n49), B=ScalarParameter() {defaultValue=2, group=Mobile Water, parameterName=Mobile Caustics Intensity})
   B=OneMinus(None=NamedRerouteUsage())
  )
  n51 = NamedRerouteUsage()
  NamedRerouteDeclaration(None=Multiply(A=Multiply(A=@n43.R, B=@n48.R), B=Multiply(A=@n50, B=@n51))) {name=Caustics, nodeColor=<rgba 1 0.9141 0 1>}
  NamedRerouteDeclaration(None=@n36.Terrain Z Height) {name=Terrain Z Height, nodeColor=<rgba 1 0 0.1961 1>}
  n52 = Constant2Vector()
  n53 = WorldPosition()
  n54 = Subtract(A=ScalarParameter() {defaultValue=-53, group=Mobile Water, parameterName=Mobile WPO Extent Height}, B=@n53.Z)
  n55 = Constant()
  n56 = WorldPosition()
  n57 = NamedRerouteUsage()
  n58 = Add(
   A=Multiply(A=Subtract(A=@n36.Water Z Height, B=@n53.Z), B=Constant3Vector() {constant=<rgba 0 0 1 0>})
   B=Multiply(
    A=MaterialFunctionCall!(Position=@n56.XYZ, Water Depth=@n57) {materialFunction=/FortWater/Materials/WaterSurface/Mobile/Functions/MF_MobileWaterWaves}.Result
    B=NamedRerouteUsage()
   )
  )
  n59 = MaterialFunctionCall!(In=StaticSwitchParameter(
    True=AppendVector(A=@n52, B=@n54)
    False=StaticSwitchParameter(True=@n55, False=@n58) {group=Mobile Water Switches, parameterName=Mobile Pool}
   ) {group=Mobile Water Switches, parameterName=Mobile Ocean Extent}) {materialFunction=/FortWater/Materials/WaterSurface/Mobile/Functions/MF_MobileBoundsClip}
  NamedRerouteDeclaration(None=@n59.Result) {name=WPO, nodeColor=<rgba 0 0.1615 0.036 1>}
  n60 = Multiply(A=LinearInterpolate(A=NamedRerouteUsage(), B=NamedRerouteUsage(), Alpha=NamedRerouteUsage()), B=NamedRerouteUsage()) {constB=0.001}
  n61 = OneMinus(None=Saturate(None=ComponentMask(None=@n60) {b=true}))
  NamedRerouteDeclaration(None=StaticSwitchParameter(True=Constant(), False=Multiply(A=@n61, B=@n61)) {group=Mobile Water Switches, parameterName=Mobile Ocean Extent}) {name=Shore Mask, nodeColor=<rgba 1 0.7801 0.0521 1>}
  NamedRerouteDeclaration(None=@n36.VelocityAndLength) {name=WaterVelocityAndScale, nodeColor=<rgba 0.8984 0 1 1>}
  n62 = Constant3Vector() {constant=<rgba 0 0 1 0>}
  n63 = Divide(
   A=ComponentMask(None=WorldPosition() {worldPositionShaderOffset=WPT_ExcludeAllShaderOffsets}.XYZ) {g=true, r=true}
   B=ScalarParameter() {defaultValue=2000, group=Mobile Water, parameterName=MobileRiver Normal Scale}
  )
  n64 = ComponentMask(None=NamedRerouteUsage()) {g=true, r=true}
  n65 = NamedRerouteUsage()
  n66 = ScalarParameter() {defaultValue=0.003, group=Mobile Water, parameterName=MobileRiver Flow Distortion Size}
  n67 = OneMinus(None=Saturate(None=Multiply(A=@n65, B=@n66) {constB=0.003}))
  n68 = Multiply(A=@n67, B=@n67)
  n69 = NamedRerouteUsage()
  n70 = Multiply(A=VertexInterpolator(VS=OneMinus(None=@n68)).PS, B=OneMinus(None=Multiply(A=@n69) {constB=0.2}))
  n71 = Multiply(
   A=Divide(A=Multiply(A=@n64, B=@n70), B=ScalarParameter() {defaultValue=2000, group=Mobile Water, parameterName=MobileRiver Normal Scale})
   B=ScalarParameter() {defaultValue=3, group=Mobile Water, parameterName=MobileRiver Normal Flow Mult}
  )
  n72 = MaterialFunctionCall(Speed=Constant() {r=0.5}) {materialFunction=&TimeWithSpeedVariable}
  n73 = Multiply(A=@n71, B=Frac(None=Add(A=@n72.Time) {constB=-0.5}))
  n74 = TextureSample(UVs=NamedRerouteDeclaration(None=Subtract(A=@n63, B=@n73)) {name=River UV1, nodeColor=<rgba 0.8984 1 0 1>}) {samplerType=SAMPLERTYPE_LinearColor, texture=/FortWater/Materials/WaterSurface/Mobile/Textures/T_MobileWater_RiverNormal}
  n75 = Add(A=AppendVector(A=@n74.R, B=@n74.R)) {constB=-0.5}
  n76 = NamedRerouteUsage()
  n77 = Frac(None=@n72.Time)
  n78 = Subtract(A=@n63, B=Multiply(A=@n71, B=@n77))
  n79 = ComponentMask(None=@n78) {g=true}
  n80 = ComponentMask(None=@n78) {r=true}
  n81 = TextureSample(UVs=NamedRerouteDeclaration(None=AppendVector(A=@n79, B=@n80)) {name=River UV2, nodeColor=<rgba 1 0 0.5859 1>}) {samplerType=SAMPLERTYPE_LinearColor, texture=&T_MobileWater_RiverNormal}
  n82 = Add(A=AppendVector(A=@n81.G, B=@n81.G)) {constB=-0.5}
  n83 = NamedRerouteUsage()
  n84 = Multiply(A=OneMinus(None=@n74.B), B=NamedRerouteUsage())
  n85 = Multiply(A=OneMinus(None=@n81.B), B=NamedRerouteUsage())
  n86 = AppendVector(A=Add(A=Multiply(A=@n75, B=@n76), B=Multiply(A=@n82, B=@n83)), B=Multiply(A=OneMinus(None=@n84), B=OneMinus(None=@n85)))
  n87 = ScalarParameter() {defaultValue=0.6, group=Mobile Water, parameterName=MobileRiver Normal Specular Intensity}
  NamedRerouteDeclaration(None=Normalize(VectorInput=LinearInterpolate(A=@n62, B=@n86, Alpha=@n87) {constAlpha=0.4})) {name=Specular Normals River, nodeColor=<rgba 0.6406 1 0 1>}
  n88 = ScalarParameter() {defaultValue=0.3, group=Mobile Water, parameterName=MobileRiver Normal Reflection Intensity}
  NamedRerouteDeclaration(None=Normalize(VectorInput=LinearInterpolate(A=@n62, B=@n86, Alpha=@n88) {constAlpha=0.06})) {name=Reflection Normals River, nodeColor=<rgba 0 1 0.0183 1>}
  n89 = NamedRerouteUsage()
  n90 = NamedRerouteUsage()
  n91 = NamedRerouteUsage()
  n92 = LinearInterpolate(A=NamedRerouteUsage(), B=NamedRerouteUsage(), Alpha=NamedRerouteUsage())
  n93 = NamedRerouteUsage()
  n94 = Constant3Vector() {constant=<rgba 0 0 1 0>}
  n95 = StaticSwitchParameter(
   True=NamedRerouteUsage()
   False=StaticSwitchParameter(
    True=NamedRerouteUsage()
    False=StaticSwitchParameter(True=@n93, False=@n94) {defaultValue=true, group=Mobile Water Switches, parameterName=Mobile Lake}
   ) {group=Mobile Water Switches, parameterName=Mobile River}
  ) {group=Mobile Water Switches, parameterName=Mobile Ocean}
  NamedRerouteDeclaration(None=StaticSwitchParameter(
    True=LinearInterpolate(A=@n89, B=@n90, Alpha=@n91)
    False=StaticSwitchParameter(True=@n92, False=@n95) {group=Mobile Water Transitions, parameterName=Mobile River To Lake}
   ) {group=Mobile Water Transitions, parameterName=Mobile River To Ocean}) {name=Specular Normals, nodeColor=<rgba 0.6406 1 0 1>}
  n96 = NamedRerouteUsage()
  n97 = NamedRerouteUsage()
  n98 = NamedRerouteUsage()
  n99 = LinearInterpolate(A=NamedRerouteUsage(), B=NamedRerouteUsage(), Alpha=NamedRerouteUsage())
  n100 = NamedRerouteUsage()
  n101 = Constant3Vector() {constant=<rgba 0 0 1 0>}
  n102 = StaticSwitchParameter(
   True=NamedRerouteUsage()
   False=StaticSwitchParameter(
    True=NamedRerouteUsage()
    False=StaticSwitchParameter(True=@n100, False=@n101) {defaultValue=true, group=Mobile Water Switches, parameterName=Mobile Lake}
   ) {group=Mobile Water Switches, parameterName=Mobile River}
  ) {group=Mobile Water Switches, parameterName=Mobile Ocean}
  NamedRerouteDeclaration(None=StaticSwitchParameter(
    True=LinearInterpolate(A=@n96, B=@n97, Alpha=@n98)
    False=StaticSwitchParameter(True=@n99, False=@n102) {group=Mobile Water Transitions, parameterName=Mobile River To Lake}
   ) {group=Mobile Water Transitions, parameterName=Mobile River To Ocean}) {name=Reflection Normals, nodeColor=<rgba 0 1 0.0183 1>}
  n103 = ScalarParameter() {defaultValue=0.0015, group=Mobile Water, parameterName=MobileOcean Shore Depth}
  n104 = ScalarParameter() {defaultValue=0.0025, group=Mobile Water, parameterName=MobileRiver Shore Depth}
  n105 = NamedRerouteUsage()
  n106 = LinearInterpolate(
   A=ScalarParameter() {defaultValue=0.0015, group=Mobile Water, parameterName=MobileLake Shore Depth}
   B=ScalarParameter() {defaultValue=0.0025, group=Mobile Water, parameterName=MobileRiver Shore Depth}
   Alpha=NamedRerouteUsage()
  )
  n107 = ScalarParameter() {defaultValue=0.0015, group=Mobile Water, parameterName=MobileLake Shore Depth}
  n108 = Constant() {r=0.001}
  n109 = StaticSwitchParameter(
   True=ScalarParameter() {defaultValue=0.0015, group=Mobile Water, parameterName=MobileOcean Shore Depth}
   False=StaticSwitchParameter(
    True=ScalarParameter() {defaultValue=0.0025, group=Mobile Water, parameterName=MobileRiver Shore Depth}
    False=StaticSwitchParameter(True=@n107, False=@n108) {defaultValue=true, group=Mobile Water Switches, parameterName=Mobile Lake}
   ) {group=Mobile Water Switches, parameterName=Mobile River}
  ) {group=Mobile Water Switches, parameterName=Mobile Ocean}
  NamedRerouteDeclaration(None=StaticSwitchParameter(
    True=LinearInterpolate(A=@n103, B=@n104, Alpha=@n105)
    False=StaticSwitchParameter(True=@n106, False=@n109) {group=Mobile Water Transitions, parameterName=Mobile River To Lake}
   ) {group=Mobile Water Transitions, parameterName=Mobile River To Ocean}) {name=Shore Depth, nodeColor=<rgba 0 0.6484 1 1>}
  n110 = ScalarParameter() {defaultValue=-2500, parameterName=Mobile WPO Ocean}
  n111 = ScalarParameter() {defaultValue=-500, parameterName=Mobile WPO River}
  n112 = NamedRerouteUsage()
  n113 = LinearInterpolate(
   A=ScalarParameter() {defaultValue=-500, parameterName=Mobile WPO Lake}
   B=ScalarParameter() {defaultValue=-500, parameterName=Mobile WPO River}
   Alpha=NamedRerouteUsage()
  )
  n114 = ScalarParameter() {defaultValue=-500, parameterName=Mobile WPO Lake}
  n115 = Constant()
  n116 = StaticSwitchParameter(
   True=ScalarParameter() {defaultValue=-2500, parameterName=Mobile WPO Ocean}
   False=StaticSwitchParameter(
    True=ScalarParameter() {defaultValue=-500, parameterName=Mobile WPO River}
    False=StaticSwitchParameter(True=@n114, False=@n115) {defaultValue=true, group=Mobile Water Switches, parameterName=Mobile Lake}
   ) {group=Mobile Water Switches, parameterName=Mobile River}
  ) {group=Mobile Water Switches, parameterName=Mobile Ocean}
  NamedRerouteDeclaration(None=StaticSwitchParameter(
    True=LinearInterpolate(A=@n110, B=@n111, Alpha=@n112)
    False=StaticSwitchParameter(True=@n113, False=@n116) {group=Mobile Water Transitions, parameterName=Mobile River To Lake}
   ) {group=Mobile Water Transitions, parameterName=Mobile River To Ocean}) {name=WPO Offset, nodeColor=<rgba 0 1 0.1094 1>}
  n117 = Multiply(A=Abs(None=NamedRerouteUsage())) {constB=0.005}
  n118 = ComponentMask(None=@n117) {r=true}
  n119 = ComponentMask(None=@n117) {g=true}
  NamedRerouteDeclaration(None=Saturate(None=Add(A=@n118, B=@n119))) {name=River Mask, nodeColor=<rgba 0.8984 0 1 1>}
  NamedRerouteDeclaration(None=@n32.TransitionMask) {name=Transition Lake to Ocean, nodeColor=<rgba 0.5234 0 1 1>}
  n120 = ComponentMask(None=WorldPosition().XYZ) {g=true, r=true}
  n121 = VectorParameter() {group=Biomes, parameterName=Foam Mask Location Radius, defaultValue=<rgba 81160 -98924 10000 1>}
  n122 = ComponentMask(None=@n121.RGB) {g=true, r=true}
  n123 = ScalarParameter() {defaultValue=0.85, group=Biomes, parameterName=Foam Mask Hardness}
  NamedRerouteDeclaration(None=OneMinus(None=SphereMask(A=@n120, B=@n122, Radius=@n121.B, Hardness=@n123) {desc=Custom Foam Mask})) {name=CustomFoamMask, nodeColor=<rgba 1 0.8672 0 1>}
  n124 = Multiply(A=Abs(None=Add(A=@n77) {constB=-0.5})) {constB=2}
  NamedRerouteDeclaration(None=OneMinus(None=NamedRerouteDeclaration(None=@n124) {name=Flow Mask, nodeColor=<rgba 0 0.0859 1 1>})) {name=Flow Counter Mask, nodeColor=<rgba 1 0.1875 0 1>}
  n125 = NamedRerouteUsage()
  n126 = NamedRerouteUsage()
  n127 = NamedRerouteUsage()
  n128 = NamedRerouteUsage()
  n129 = WorldPosition() {worldPositionShaderOffset=WPT_ExcludeAllShaderOffsets}
  n130 = Multiply(A=TextureSample(UVs=Multiply(A=@n129.XY) {constB=0.0001}) {samplerType=SAMPLERTYPE_LinearColor, texture=/FortWater/Materials/WaterSurface/Mobile/Textures/T_MobileWater_Foam}.R) {constB=1.4}
  n131 = Multiply(
   A=Add(A=Multiply(A=@n125, B=@n126), B=Multiply(A=@n127, B=@n128))
   B=NamedRerouteDeclaration(None=Saturate(None=@n130)) {name=River Static Foam, nodeColor=<rgba 0.4062 1 0 1>}
  )
  n132 = Abs(None=NamedRerouteUsage())
  n133 = Subtract(
   A=Add(A=ComponentMask(None=@n132) {r=true}, B=ComponentMask(None=@n132) {g=true})
   B=ScalarParameter() {defaultValue=400, group=Mobile Water, parameterName=MobileRiver Foam Height}
  ) {constB=260}
  n134 = ScalarParameter() {defaultValue=0.01, group=Mobile Water, parameterName=MobileRiver Foam Fade}
  n135 = NamedRerouteUsage()
  n136 = ScalarParameter() {defaultValue=0.0025, group=Mobile Water, parameterName=MobileRiver Foam Distance}
  n137 = OneMinus(None=Saturate(None=Multiply(A=@n135, B=@n136) {constB=0.001}))
  n138 = Max(A=Saturate(None=Multiply(A=@n133, B=@n134)), B=Multiply(A=@n137, B=@n137))
  n139 = Saturate(None=Multiply(A=Multiply(A=@n131, B=@n138)) {constB=4})
  n140 = Multiply(A=Multiply(A=@n139, B=@n139), B=Multiply(A=@n131) {constB=1.5})
  n141 = ScalarParameter() {defaultValue=0.75, group=Mobile Water, parameterName=MobileRiver Foam Opacity}
  NamedRerouteDeclaration(None=StaticSwitchParameter(True=Constant(), False=Multiply(A=@n140, B=@n141)) {group=Mobile Water Transitions, parameterName=Disable RiverFoam}) {name=Foam Mask River, nodeColor=<rgba 0.9441 1 0.8079 1>}
  n142 = Multiply(
   A=WorldPosition() {worldPositionShaderOffset=WPT_ExcludeAllShaderOffsets}.XY
   B=ScalarParameter() {defaultValue=0.0006, group=Mobile Water, parameterName=MobileLake Foam Scale}
  ) {constB=0.0001}
  n143 = Multiply(A=Time(), B=ScalarParameter() {defaultValue=0.005, group=Mobile Water, parameterName=MobileLake Foam Speed}) {constB=0.03}
  n144 = Multiply(
   A=NamedRerouteUsage()
   B=TextureSample(UVs=Add(A=@n142, B=@n143)) {samplerType=SAMPLERTYPE_LinearColor, texture=&T_MobileWater_Foam}.R
  )
  n145 = NamedRerouteUsage()
  n146 = ScalarParameter() {defaultValue=0.005, group=Mobile Water, parameterName=MobileLake Foam Distance}
  n147 = OneMinus(None=Saturate(None=Multiply(A=@n145, B=@n146) {constB=0.001}))
  n148 = Multiply(A=@n144, B=Multiply(A=@n147, B=@n147))
  n149 = Multiply(A=Subtract(A=Multiply(A=@n148) {constB=2}) {constB=0.1}, B=Multiply(A=@n144) {constB=1.2})
  NamedRerouteDeclaration(None=Multiply(A=Saturate(None=@n149), B=ScalarParameter() {defaultValue=0.5, group=Mobile Water, parameterName=MobileLake Foam Opacity}) {constB=0.4}) {name=Foam Mask Lake, nodeColor=<rgba 0.9441 1 0.8079 1>}
  n150 = NamedRerouteUsage()
  n151 = NamedRerouteUsage()
  n152 = NamedRerouteUsage()
  n153 = NamedRerouteUsage()
  n154 = NamedRerouteUsage()
  n155 = StaticSwitchParameter(
   True=NamedRerouteUsage()
   False=StaticSwitchParameter(True=NamedRerouteUsage(), False=Constant() {r=0.1}) {defaultValue=true, group=Mobile Water Switches, parameterName=Mobile Lake}
  ) {group=Mobile Water Switches, parameterName=Mobile River}
  n156 = StaticSwitchParameter(
   True=LinearInterpolate(A=NamedRerouteUsage(), B=NamedRerouteUsage(), Alpha=NamedRerouteUsage())
   False=StaticSwitchParameter(
    True=LinearInterpolate(A=@n151, B=@n152, Alpha=@n153)
    False=StaticSwitchParameter(True=@n154, False=@n155) {group=Mobile Water Switches, parameterName=Mobile Ocean}
   ) {group=Mobile Water Transitions, parameterName=Mobile River To Lake}
  ) {group=Mobile Water Transitions, parameterName=Mobile River To Ocean}
  NamedRerouteDeclaration(None=StaticSwitchParameter(True=Multiply(A=@n150, B=@n156), False=@n156) {parameterName=UseCustomFoamMask}) {name=Foam Mask, nodeColor=<rgba 0.9441 1 0.8079 1>}
  NamedRerouteDeclaration(None=@n74.A) {name=Foam River Texture 1, nodeColor=<rgba 0 0.5078 1 1>}
  NamedRerouteDeclaration(None=@n81.A) {name=Foam River Texture 2, nodeColor=<rgba 1 0.3047 0 1>}
  n157 = VectorParameter() {group=Mobile Water, parameterName=Mobile Color Deep, sortPriority=1, defaultValue=<rgba 0.0269 0.1104 0.2344 1>}
  n158 = MaterialFunctionCall!(Island Mask=NamedRerouteUsage()) {materialFunction=/FortWater/Materials/WaterSurface/Mobile/Functions/MF_MobileOceanColorMask}
  NamedRerouteDeclaration(None=StaticSwitchParameter(True=Add(A=@n157.RGB, B=@n158.Mask Color), False=@n157.RGB) {parameterName=OceanColorMask}) {name=Color Deep, nodeColor=<rgba 0.0026 0.0135 0.026 1>}
  n159 = WorldPosition() {worldPositionShaderOffset=WPT_ExcludeAllShaderOffsets}
  n160 = ScalarParameter() {defaultValue=0.0004, group=Mobile Water, parameterName=MobileLake Normal1 Scale}
  n161 = Time()
  n162 = ScalarParameter() {defaultValue=0.03, group=Mobile Water, parameterName=MobileLake Normal1 Speed}
  n163 = TextureSample(UVs=Add(A=Multiply(A=@n159.XY, B=@n160) {constB=0.0001}, B=Multiply(A=@n161, B=@n162) {constB=0.03})) {samplerType=SAMPLERTYPE_Normal, texture=/FortWater/Materials/WaterSurface/Mobile/Textures/T_MobileWater_LakeNormal}
  n164 = Multiply(A=OneMinus(None=Saturate(None=@n163.R)), B=OneMinus(None=Saturate(None=@n163.G)))
  NamedRerouteDeclaration(None=Multiply(A=@n164, B=@n164)) {name=Foam Lake Texture 1, nodeColor=<rgba 1 0.2578 0 1>}
  n165 = WorldPosition() {worldPositionShaderOffset=WPT_ExcludeAllShaderOffsets}
  n166 = AppendVector(A=ComponentMask(None=@n165.XY) {g=true}, B=ComponentMask(None=@n165.XY) {r=true})
  n167 = ScalarParameter() {defaultValue=0.0003, group=Mobile Water, parameterName=MobileLake Normal2 Scale}
  n168 = Time()
  n169 = AppendVector(
   A=ScalarParameter() {defaultValue=0.04, group=Mobile Water, parameterName=MobileLake Normal2 Speed X}
   B=ScalarParameter() {defaultValue=-0.05, group=Mobile Water, parameterName=MobileLake Normal2 Speed Y}
  )
  n170 = TextureSample(UVs=Subtract(A=Multiply(A=@n166, B=@n167) {constB=0.0001}, B=Multiply(A=@n168, B=@n169) {constB=0.026})) {samplerType=SAMPLERTYPE_Normal, texture=&T_MobileWater_LakeNormal}
  NamedRerouteDeclaration(None=@n170.A) {name=Foam Lake Texture 2, nodeColor=<rgba 0.0547 1 0 1>}
  n171 = Multiply(
   A=WorldPosition() {worldPositionShaderOffset=WPT_ExcludeAllShaderOffsets}.XY
   B=ScalarParameter() {defaultValue=0.0005, group=Mobile Water, parameterName=MobileOcean Foam Scale}
  ) {constB=0.0001}
  n172 = Multiply(A=Time(), B=ScalarParameter() {defaultValue=0.01, group=Mobile Water, parameterName=MobileOcean Foam Speed}) {constB=0.03}
  n173 = Multiply(
   A=NamedRerouteUsage()
   B=TextureSample(UVs=Add(A=@n171, B=@n172)) {samplerType=SAMPLERTYPE_LinearColor, texture=&T_MobileWater_Foam}.R
  )
  n174 = NamedRerouteUsage()
  n175 = ScalarParameter() {defaultValue=0.0013, group=Mobile Water, parameterName=MobileOcean Foam Distance}
  n176 = Saturate(None=Multiply(A=Multiply(A=@n174, B=@n175) {constB=0.001}))
  n177 = OneMinus(None=@n176)
  n178 = Multiply(A=Multiply(A=@n173, B=Multiply(A=@n177, B=@n177))) {constB=3}
  n179 = Multiply(A=Saturate(None=Subtract(A=@n178) {constB=0.2}), B=Multiply(A=@n173) {constB=1.3})
  n180 = ScalarParameter() {defaultValue=1, group=Mobile Water, parameterName=MobileOcean Foam Opacity}
  NamedRerouteDeclaration(None=StaticSwitchParameter(True=Constant(), False=Multiply(A=@n179, B=@n180) {constB=0.4}) {group=Mobile Water Switches, parameterName=Mobile Ocean Extent}) {name=Foam Mask Ocean, nodeColor=<rgba 0.9441 1 0.8079 1>}
  n181 = Constant3Vector() {constant=<rgba 0 0 1 0>}
  n182 = Multiply(
   A=WorldPosition() {worldPositionShaderOffset=WPT_ExcludeAllShaderOffsets}.XY
   B=ScalarParameter() {defaultValue=0.0003, group=Mobile Water, parameterName=MobileOcean Normal1 Scale}
  ) {constB=0.0001}
  n183 = Multiply(A=Time(), B=ScalarParameter() {defaultValue=0.015, group=Mobile Water, parameterName=MobileOcean Normal1 Speed}) {constB=0.03}
  n184 = MaterialFunctionCall(Float3=TextureSample(UVs=Add(A=@n182, B=@n183)) {samplerType=SAMPLERTYPE_Normal, texture=&T_MobileWater_LakeNormal}.RGB) {materialFunction=&BreakOutFloat3Components}
  n185 = AppendVector(A=@n184.R, B=@n184.R)
  n186 = WorldPosition() {worldPositionShaderOffset=WPT_ExcludeAllShaderOffsets}
  n187 = AppendVector(A=ComponentMask(None=@n186.XY) {g=true}, B=ComponentMask(None=@n186.XY) {r=true})
  n188 = ScalarParameter() {defaultValue=0.0001, group=Mobile Water, parameterName=MobileOcean Normal2 Scale}
  n189 = Time()
  n190 = AppendVector(
   A=ScalarParameter() {defaultValue=0.04, group=Mobile Water, parameterName=MobileOcean Normal2 Speed X}
   B=ScalarParameter() {defaultValue=-0.05, group=Mobile Water, parameterName=MobileOcean Normal2 Speed Y}
  )
  n191 = TextureSample(UVs=Subtract(A=Multiply(A=@n187, B=@n188) {constB=0.0001}, B=Multiply(A=@n189, B=@n190) {constB=0.026})) {samplerType=SAMPLERTYPE_Normal, texture=&T_MobileWater_LakeNormal}
  n192 = AppendVector(A=@n191.G, B=@n191.G)
  n193 = WorldPosition() {worldPositionShaderOffset=WPT_ExcludeAllShaderOffsets}
  n194 = ScalarParameter() {defaultValue=0, group=Mobile Water, parameterName=MobileOcean Normal3 Scale}
  n195 = Time()
  n196 = ScalarParameter() {defaultValue=0.01, group=Mobile Water, parameterName=MobileOcean Normal3 Speed}
  n197 = TextureSample(UVs=Add(A=Multiply(A=@n193.XY, B=@n194) {constB=0.0001}, B=Multiply(A=@n195, B=@n196) {constB=0.03})) {samplerType=SAMPLERTYPE_Normal, texture=&T_MobileWater_LakeNormal}
  n198 = Constant3Vector() {constant=<rgba 0 0 1 0>}
  n199 = NamedRerouteUsage()
  n200 = MaterialFunctionCall(Float3=StaticSwitchParameter(True=LinearInterpolate(A=@n197.RGB, B=@n198, Alpha=@n199), False=@n197.RGB) {group=Styx, parameterName=Is Mobile Lake To Ocean Transiiton}) {materialFunction=&BreakOutFloat3Components}
  n201 = AppendVector(A=Add(A=Add(A=@n185, B=@n192), B=AppendVector(A=@n200.R, B=@n200.R)), B=Multiply(A=Multiply(A=@n184.B, B=@n191.B), B=@n200.B))
  n202 = ScalarParameter() {defaultValue=0.4, group=Mobile Water, parameterName=MobileOcean Normal Specular Intensity}
  NamedRerouteDeclaration(None=Normalize(VectorInput=LinearInterpolate(A=@n181, B=@n201, Alpha=@n202) {constAlpha=0.4})) {name=Specular Normals Ocean, nodeColor=<rgba 0.6406 1 0 1>}
  n203 = ScalarParameter() {defaultValue=0.1, group=Mobile Water, parameterName=MobileOcean Normal Reflection Intensity}
  NamedRerouteDeclaration(None=Normalize(VectorInput=LinearInterpolate(A=@n181, B=@n201, Alpha=@n203) {constAlpha=0.06})) {name=Reflection Normals Ocean, nodeColor=<rgba 0 1 0.0183 1>}
  n204 = MaterialFunctionCall(
   Color_Morning=VectorParameter() {parameterName=Sky Color Morning, defaultValue=<rgba 1 0.5427 0.3698 1>}.RGB
   Color_Day=VectorParameter() {parameterName=Sky Color Day, defaultValue=<rgba 1 1 1 1>}.RGB
   Color_Evening=VectorParameter() {parameterName=Sky Color Evening, defaultValue=<rgba 0.9635 0.4417 0.3162 1>}.RGB
   Color_Night=VectorParameter() {parameterName=Sky Color Night, defaultValue=<rgba 0.2484 0.3134 0.6562 1>}.RGB
  ) {materialFunction=&Color_Time_of_Day}
  n205 = VectorParameter() {group=Mobile Water, parameterName=Mobile Color Shallow, sortPriority=5, defaultValue=<rgba 0.0132 0.2787 0.3177 1>}
  n206 = LinearInterpolate(Alpha=ComponentMask(None=SkyAtmosphereLightDirection()) {b=true}) {constA=0.95}
  NamedRerouteDeclaration(None=VertexInterpolator(VS=LinearInterpolate(A=@n204.Result, B=@n205.RGB, Alpha=@n206)).PS) {name=Color Shallow, nodeColor=<rgba 0.6413 0.8456 1 1>}
  n207 = Constant3Vector() {constant=<rgba 0 0 1 0>}
  n208 = AppendVector(A=Add(A=AppendVector(A=@n163.R, B=@n163.R), B=AppendVector(A=@n170.G, B=@n170.G)), B=Multiply(A=@n163.B, B=@n170.B))
  n209 = ScalarParameter() {defaultValue=0.4, group=Mobile Water, parameterName=MobileLake Normal Specular Intensity}
  NamedRerouteDeclaration(None=Normalize(VectorInput=LinearInterpolate(A=@n207, B=@n208, Alpha=@n209) {constAlpha=0.4})) {name=Specular Normals Lake, nodeColor=<rgba 0.6406 1 0 1>}
  n210 = Multiply(A=OneMinus(None=Saturate(None=@n184.R)), B=OneMinus(None=Saturate(None=@n184.G)))
  NamedRerouteDeclaration(None=Multiply(A=@n210, B=@n210)) {name=Foam Ocean Texture 1, nodeColor=<rgba 1 0.2578 0 1>}
  NamedRerouteDeclaration(None=@n191.A) {name=Foam Ocean Texture 2, nodeColor=<rgba 0.0547 1 0 1>}
  NamedRerouteDeclaration(None=@n33.MobileWaterColor) {name=Gold_WaterColor, nodeColor=<rgba 0 0.9766 1 1>}
  NamedRerouteDeclaration(None=VertexInterpolator(VS=@n204.Result).PS) {name=Sky Color, nodeColor=<rgba 0.7578 1 0 1>}
  NamedRerouteDeclaration(None=@n33.SphereMask) {name=SphereMaskGold, nodeColor=<rgba 0.125 0 1 1>}
  n211 = Subtract(
   A=Distance(A=WorldPosition() {worldPositionShaderOffset=WPT_ExcludeAllShaderOffsets}.XY, B=Constant())
   B=ScalarParameter() {defaultValue=134205.6406, parameterName=Mobile Island Extent}
  ) {constB=102000}
  n212 = Saturate(None=Divide(A=Multiply(A=@n211) {constB=-1}, B=ScalarParameter() {defaultValue=58994.9023, parameterName=Mobile Island Fade}) {constB=100000})
  NamedRerouteDeclaration(None=Multiply(A=@n212, B=@n212)) {name=Island Mask, nodeColor=<rgba 0.8023 1 0.4533 1>}
  n213 = Multiply(A=Subtract(A=PixelDepth().R, B=ScalarParameter() {defaultValue=20027.0469, parameterName=Mobile Camera Offset}) {constB=102000}) {constB=-1}
  n214 = ScalarParameter() {defaultValue=12689.7812, parameterName=Mobile Camera Fade}
  NamedRerouteDeclaration(None=Saturate(None=Divide(A=@n213, B=@n214) {constB=100000})) {name=Camera Fade, nodeColor=<rgba 0 0.0391 1 1>}
  n215 = ScalarParameter() {defaultValue=0.1, group=Mobile Water, parameterName=MobileLake Normal Reflection Intensity}
  NamedRerouteDeclaration(None=Normalize(VectorInput=LinearInterpolate(A=@n207, B=@n208, Alpha=@n215) {constAlpha=0.06})) {name=Reflection Normals Lake, nodeColor=<rgba 0 1 0.0183 1>}
  n216 = SkyAtmosphereLightDirection()
  n217 = ComponentMask(None=@n216) {b=true}
  n218 = AppendVector(A=ComponentMask(None=@n216) {g=true, r=true}, B=LinearInterpolate(Alpha=@n217) {constA=0.5})
  n219 = CameraVectorWS()
  n220 = DotProduct(A=Normalize(VectorInput=Add(A=@n218, B=@n219)), B=NamedRerouteUsage())
  n221 = Multiply(
   A=Power(Base=@n220) {constExponent=128}
   B=ScalarParameter() {defaultValue=0.04, group=Mobile Water, parameterName=Mobile Specular Diffusion}
  ) {constB=0.02}
  n222 = OneMinus(None=@n220)
  n223 = OneMinus(None=Saturate(None=Multiply(A=@n222) {constB=500}))
  n224 = Multiply(A=@n223, B=@n223)
  n225 = Multiply(A=@n224, B=@n224)
  n226 = Multiply(A=@n225, B=@n225)
  n227 = Multiply(A=OneMinus(None=@n217)) {constB=3}
  n228 = Add(A=Multiply(A=@n227, B=@n227))
  n229 = Desaturation(None=NamedRerouteUsage(), Fraction=Constant() {r=-1})
  NamedRerouteDeclaration(None=Multiply(A=Add(A=@n221, B=@n226), B=Multiply(A=@n228, B=@n229)) {constB=2}) {name=Sun Specular, nodeColor=<rgba 1 0.1699 0 1>}
  n230 = ComponentMask(None=Multiply(A=CameraVectorWS()) {constB=-1}) {g=true, r=true}
  n231 = Constant() {r=0.01}
  n232 = DotProduct(A=Normalize(VectorInput=AppendVector(A=@n230, B=@n231)), B=NamedRerouteUsage())
  NamedRerouteDeclaration(None=Multiply(A=Saturate(None=@n232), B=ScalarParameter() {defaultValue=0.4, group=Mobile Water, parameterName=Mobile Side Reflections}) {constB=0.1}) {name=Side Reflections Mask, nodeColor=<rgba 0.0547 1 0 1>}
  n233 = NamedRerouteUsage()
  n234 = CameraVectorWS()
  n235 = Multiply(A=Saturate(None=@n1)) {constB=0.13}
  n236 = ReflectionVectorWS(CustomWorldNormal=Subtract(A=@n233, B=Multiply(A=@n234, B=@n235) {constB=0.14}) {constB=0}) {bNormalizeCustomWorldNormal=true}
  NamedRerouteDeclaration(None=TextureSample(UVs=@n236) {samplerType=SAMPLERTYPE_LinearColor, texture=/FortWater/Materials/WaterSurface/Mobile/Textures/T_MobileWater_Cubemap}.RGB) {name=Reflection Color, nodeColor=<rgba 0 0.2297 1 1>}
  n237 = DotProduct(A=@n233, B=CameraVectorWS())
  n238 = Power(
   Base=OneMinus(None=Saturate(None=@n237))
   Exp=ScalarParameter() {defaultValue=4, group=Mobile Water, parameterName=Mobile Reflection MaskPower}
  ) {constExponent=4}
  n239 = ScalarParameter() {defaultValue=0.8, group=Mobile Water, parameterName=Mobile Reflection Intensity}
  NamedRerouteDeclaration(None=Add(A=Multiply(A=@n238, B=@n239) {constB=0.75}, B=NamedRerouteUsage())) {name=Reflection Mask, nodeColor=<rgba 1 0.0984 0.7042 1>}
  NamedRerouteUsage()
  NamedRerouteUsage()
  NamedRerouteUsage()
  NamedRerouteUsage()
  NamedRerouteUsage()
  Reroute(None=@n176)
  n240 = MaterialFunctionCall!(
   Material Attributes=@n34.MaterialAttributes
   MobileEmissive=@n34.MaterialAttributes
   MobileOpacity=@n34.MaterialAttributes
   Scattering
   Absorption
   Sun Specular=NamedRerouteUsage()
   Is translucent=@n3
  ) {materialFunction=/Hera/Sets/EchoOasis/Materials/MaterialFunctions/MF_SquidWater}
  SetMaterialAttributes!(
   MaterialAttributes=@n34.MaterialAttributes
   Emissive Color=@n240.MobileEmissive
   Opacity=@n240.MobileEmissive
   Normal=@n34.MaterialAttributes
  ) {attributeSetTypes [#1, #2, #4]}
  SkyLightEnvMapSample(Direction=@n236, Roughness=Constant())
  StaticSwitchParameter(True=@n17, False=@n15) {group=Gold Liquid, parameterName=MaskFoamByGoldMask}
  n241 = Add(A=@n59.Result, B=WorldPosition() {worldPositionShaderOffset=WPT_CameraRelative}.XYZ)
  n242 = Normalize(VectorInput=@n241)
  n243 = MaterialFunctionCall() {materialFunction=&CameraDirectionVector}
  n244 = Divide(
   A=Subtract(A=DotProduct(A=@n241, B=@n243.Result), B=ScalarParameter() {defaultValue=50000, parameterName=Mobile WPO Offset}) {constB=102000}
   B=ScalarParameter() {defaultValue=12689.7812, parameterName=Mobile WPO Fade}
  ) {constB=100000}
  MaterialExpressionAbs_3 = Abs(None=WorldPosition() {worldPositionShaderOffset=WPT_ExcludeAllShaderOffsets}.XY)
  n245 = Convert(None=@MaterialExpressionAbs_3) {
   nodeName=Break Float2
   convertInputs [{
     type=Vector2
     expressionInput {
      expression=&self:MaterialExpressionAbs_3
      inputName=None
      mask=0
      maskA=0
      maskB=0
      maskG=0
      maskR=0
      outputIndex=0
     }
    }]
   convertOutputs [{type=Scalar}, {type=Scalar}]
  }
  n246 = Max(A=@n245, B=@n245)
  n247 = ScalarParameter() {defaultValue=194235.9844, parameterName=Mobile WPO Extent}
  n248 = Divide(
   A=Multiply(A=Subtract(A=@n246, B=@n247) {constB=102000}) {constB=-1}
   B=ScalarParameter() {defaultValue=57548.4375, parameterName=Mobile WPO Extent Fade}
  ) {constB=100000}
  n249 = Multiply(A=Multiply(A=Saturate(None=@n244), B=Saturate(None=@n248)), B=NamedRerouteUsage()) {constB=-2000}
  StaticSwitchParameter(True=Add(A=@n59.Result, B=Multiply(A=@n242, B=@n249) {constB=-2000}), False=@n59.Result) {parameterName=Mobile Camera Offset}
  VertexNormalWS()
 ]
}
